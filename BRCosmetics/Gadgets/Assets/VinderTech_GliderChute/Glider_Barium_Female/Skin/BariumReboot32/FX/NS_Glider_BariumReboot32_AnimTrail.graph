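# Niagara animated trail effect for the BariumReboot32 glider skin, combining light ribbons, spinning wheel meshes and speed-line sprites.
# Four ribbons and the wheels sample skeletal mesh sockets, with ribbon points lerping toward offset targets and shifted against owner velocity for lag.
# Smoothed forward input drives ribbon length and brightness, an integrator fades wheels and speed lines, and RightDot gates sideways thrust audio.
# Front-end preview flags adjust the speed lines.
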NiagaraSystem /BRCosmetics/Gadgets/Assets/VinderTech_GliderChute/Glider_Barium_Female/Skin/BariumReboot32/FX/NS_Glider_BariumReboot32_AnimTrail
dependencies {
 usedDataInterfaces [
  &NiagaraDataInterfaceVector2DCurve
  &NiagaraDataInterfaceCurve
  &NiagaraDataInterfaceColorCurve
  &NiagaraDataInterfaceAudioPlayer
  &NiagaraDataInterfaceArrayFloat
 ]
 usedDynamicInputs [&ClampFloat]
 usedModules [
  &self:SystemSpawnScript.SystemScriptSource.SystemScriptGraph.NiagaraNodeAssignment_0.SetVariables_#1
  &self:SystemSpawnScript.SystemScriptSource.SystemScriptGraph.NiagaraNodeAssignment_2.SetVariables_#2
  &self:SystemSpawnScript.SystemScriptSource.SystemScriptGraph.NiagaraNodeAssignment_3.SetVariables_#3
  &SystemState
  &FN_Integrator
  &self:SystemSpawnScript.SystemScriptSource.SystemScriptGraph.NiagaraNodeAssignment_1.SetVariables_#4
 ]
}
emitters {
 FN_Ribbon_Light {
  emitterSpawnScript [
   /SetVariables(Emitter.RibbonNumber=4) {isSetParameters=true}
   /SetVariables(Emitter.ParticlePerRibbon=10) {isSetParameters=true}
  ]
  emitterUpdateScript [
   &EmitterState(
    Inactive Response=::NewEnumerator1
    Life Cycle Mode=::NewEnumerator1
    Loop Behavior=::NewEnumerator1
    Loop Duration Mode=::NewEnumerator1
    UseLoopDelay=false
   )
   &SpawnBurst_Instantaneous(
    Spawn Count=&Multiply_Int(A=$Emitter.RibbonNumber, B=$Emitter.ParticlePerRibbon)
    Use Spawn Probability=false
    UseLoopCountLimit=false
   )
   &FN_Timer()
  ]
  particleSpawnScript [
   &InitializeParticle(
    Color=<rgba 1 0.0173 0 1>
    Lifetime=0.7
    Ribbon Width=11
    Ribbon Width Mode=::NewEnumerator1
    Sprite Size Mode=::NewEnumerator3
    UsePositionOffset=false
    Write Material Random=true
   )
   &PartitionParticles(
    ID to Partition:NiagaraInt32=&ReturnExecIndex()
    Normalized Index In Partition Includes One=true
    Number Of Partitions=$Emitter.RibbonNumber
    Particles Per Partition:NiagaraInt32=$Emitter.ParticlePerRibbon
    Partition Mode:ENiagara_PartitionMode=::NewEnumerator1
    Write Partition Index to Ribbon Link Order:ENiagara_AttributeSamplingApplyOutput=::NewEnumerator0
    Write Partition to Ribbon ID=::NewEnumerator0
   )
   &SkeletalMeshLocation(
    Bone / Socket Index=$Particles.PartitionParticles.Partition
    MaskSpawnGroup=false
    Mesh Sampling Type=::NewEnumerator5
    Orientation Sampling=::NewEnumerator1
    Sampled Position Offset=&Lerp_Vector(A=<0 0 0>, B=<0 100 0>, Alpha=$Particles.PartitionParticles.Normalized Index In Partition)
    Skeletal Mesh=$System.SkeletalMesh
    Socket Sampling Mode=::NewEnumerator5
    Use Sampled Mesh Attribute as Ribbon ID=false
   ) {module=SkeletalMeshLocation001}
  ]
  particleUpdateScript [
   &ParticleState(Kill Particles When Lifetime Has Elapsed=false)
   &SkeletalMeshLocation(
    Bone / Socket Index=$Particles.PartitionParticles.Partition
    MaskSpawnGroup=false
    Mesh Sampling Type=::NewEnumerator5
    Orientation Sampling=::NewEnumerator1
    Position Sampling=::NewEnumerator2
    Sampled Position Offset=&Lerp_Vector(
     A=<0 0 0>
     B=&FN_FrontEndVector3Modifier(Gameplay=&Lerp_Vector(A=<0 50 0>, B=<0 75 0>, Alpha=$System.SmoothedForwardDot), Front End=<0 75 0>)
     Alpha=$Particles.PartitionParticles.Normalized Index In Partition
    )
    Sampled Position Offset Coordinate Space:ENiagaraSkelMeshTransforms=::NewEnumerator2
    Skeletal Mesh=$System.SkeletalMesh
    Socket Sampling Mode=::NewEnumerator5
    Use Sampled Mesh Attribute as Ribbon ID=false
   )
   /SetVariables(Particles.GoalPos:NiagaraPosition=$Particles.SkeletalMeshLocation.SampledPosition) {isSetParameters=true}
   /SetVariables(Particles.Position=&LerpPosition(
     A=$Particles.Position
     B=$Particles.GoalPos
     Alpha=&Multiply_Float(
      A=&Multiply_Float(
       A=$Engine.DeltaTime
       B=&Multiply_Float(
        A=&Lerp_Float(
         A=10
         B=5
         Alpha=&FloatFromCurve(
          FloatCurve=DataInterface() {propertyValues {Curve {keys [<at 0 0>,<at 1 1>]}}}
          CurveIndex=$Particles.PartitionParticles.Normalized Index In Partition
         )
        )
        B=1
       )
      )
      B=1
     )
    )) {isSetParameters=true}
   &InheritSourceMovement(
    Applied Position Scale=&Multiply_VectorByFloat(
     Vector=<-0.25 -0.25 -0.1>
     Float=&FloatFromCurve(
      FloatCurve=DataInterface() {propertyValues {Curve {keys [
          {arriveTangent=0.4151, interpMode=RCIM_Cubic, leaveTangent=0.4151, tangentMode=RCTM_User, time=0, value=0}
          {arriveTangent=2.9414, interpMode=RCIM_Cubic, leaveTangent=2.9414, tangentMode=RCTM_User, time=1, value=1}
         ]}}}
      CurveIndex=$Particles.PartitionParticles.Normalized Index In Partition
     )
    )
    ApplyForce=false
    ApplyPosition=true
    ApplyPositionFalloff=false
    Source Velocity:Vector3f=&TransformVector(Vector=$Engine.Owner.Velocity, Source Space=::World, Destination Space=::Local)
   ) {module=InheritSourceMovement001}
   &SolveForcesAndVelocity(Clamp Velocity=false, Limit Acceleration=false)
   &DynamicMaterialParameters(
    Index 3 Param 1=&Add_Float(
     A=&SelectFloatFromArray(Direct Array Index=$Particles.SpawnGroup, Float Selection Array=DataInterface() {propertyValues {FloatData [1,2,3,4]}})
     B=$Emitter.FN_Timer.Value
    )
    Module.Param0WriteEnabled3=true
    Module.Param1WriteEnabled3=false
    Module.Param2WriteEnabled3=false
    Module.Param3WriteEnabled3=false
    Write Parameter Index 0=false
    Write Parameter Index 3=true
   )
   &ScaleColor(
    Scale Alpha=&FloatFromCurve(
     FloatCurve=DataInterface() {propertyValues {Curve {keys [
         {interpMode=RCIM_Cubic, tangentMode=RCTM_User, time=0, value=1}
         {arriveTangent=-1.7006, interpMode=RCIM_Cubic, leaveTangent=-1.7006, tangentMode=RCTM_User, time=1, value=0}
        ]}}}
     CurveIndex=$Particles.PartitionParticles.Normalized Index In Partition
     Scale Curve=&FloatFromCurve(FloatCurve=DataInterface() {propertyValues {Curve {keys [<at 1 0>,<at 2 1>]}}}, CurveIndex=$Emitter.Age)
    )
    Scale RGB=&VectorFromFloat(Value=&Lerp_Float(A=1, B=2, Alpha=$System.SmoothedForwardDot))
    ScaleA=true
    ScaleRGB=true
   )
   /SetVariables(Particles.SecondRibbonCoor:LinearColor=&MultiplyLinearColor(
     Linear Color=<rgba 0.15 0.0097 0 1>
     Scale Factor=&MakeVector4_FromLinearColor(LinearColor=$Particles.Color)
    )) {isSetParameters=true}
  ]
  rendererClasses [&NiagaraRibbonRendererProperties]
  renderers [
   {rendererClass=&NiagaraRibbonRendererProperties}
   {rendererClass=&NiagaraRibbonRendererProperties, rendererIndex=1}
  ]
 }
 NE_Wheels {
  emitterSpawnScript [/SetVariables(Emitter.AudioPlayer=DataInterface() {propertyValues {
      Attenuation=null
      Concurrency=null
      MaxPlaysPerTick=10
      SoundToPlay=/BRCosmetics/Sounds/Gliders/BariumFemale/GliderThrustStart_BariumFemale_Sideways_Cue
      bAllowLoopingOneShotSounds=false
      bLimitPlaysPerTick=true
      bOnlyActiveDuringGameplay=false
      bStopWhenComponentIsDestroyed=true
      ConfigurationUserParameter {parameter {name=None, typeDefHandle {registeredTypeIndex=1}, varData []}}
      ParameterNames []
     }}) {isSetParameters=true}]
  emitterUpdateScript [
   &EmitterState(Life Cycle Mode=::NewEnumerator1, Loop Behavior=::NewEnumerator1, Loop Duration Mode=::NewEnumerator1, UseLoopDelay=true)
   &FN_Timer()
   &SpawnBurst_Instantaneous(Spawn Count=4, Use Spawn Probability=false, UseLoopCountLimit=false)
   /Niagara/Modules/Audio/PlayPersistentAudio(
    Audio Player:NiagaraDataInterfaceAudioPlayer=$Emitter.AudioPlayer
    Enable Fade In / Out=true
    Fade Out Time=1.5
    Play Audio:NiagaraBool=&SetBoolByFloatComparison(A=&Abs_Float(Float:NiagaraFloat=$User.RightDot), B=0.5)
   )
   /Niagara/Modules/Audio/UpdatePersistentAudio(Audio Player:NiagaraDataInterfaceAudioPlayer=$Emitter.AudioPlayer, Update Type:ENiagara_AudioParamType=::NewEnumerator5)
   &SpawnBurst_Instantaneous(Spawn Count=4, Spawn Time=1.9833, Use Spawn Probability=false, UseLoopCountLimit=false) {module=SpawnBurst_Instantaneous001}
  ]
  particleSpawnScript [
   &FN_SetSpawnGroup()
   &InitializeParticle(Color=<rgba 0.05 0 0 5>, Mesh Scale=<0.3 0.3 0.3>, Mesh Scale Mode=::NewEnumerator4, UsePositionOffset=false, Write Material Random=true)
  ]
  particleUpdateScript [
   &ParticleState(Kill Particles When Lifetime Has Elapsed=false, Let Infinitely Lived Particles Die When Emitter Deactivates=true)
   &SkeletalMeshLocation(
    Bone / Socket Index=$Particles.SpawnGroup
    MaskSpawnGroup=false
    Mesh Sampling Type=::NewEnumerator5
    Skeletal Mesh=$System.SkeletalMesh
    Socket Sampling Mode=::NewEnumerator5
   )
   /SetVariables(Particles.MeshOrientation=&MultiplyQuaternion(
     Quaternion A=&MultiplyQuaternion(
      Quaternion A=$Particles.SkeletalMeshLocation.SampledRotation
      Quaternion B=&MakeQuatFromAxisAngle(Axis=<1 0 0>, AngleInDegrees=&RandomRangeFloat(Minimum=-360, Maximum=360))
     )
     Quaternion B=&MakeQuatFromAxisAngle(Axis=<1 0 0>, AngleInDegrees=&Multiply_Float(A=360, B=$Engine.Time), NormalizeAxis=true)
    )) {isSetParameters=true}
   &DynamicMaterialParameters(
    Index 2 Param 4=1
    Index 3 Param 1=&Add_Float(A=&RandomRangeFloat(Minimum=0, Maximum=1), B=$Emitter.FN_Timer.Value)
    Index 3 Param 2=0.13
    Module.Param0WriteEnabled2=true
    Module.Param0WriteEnabled3=true
    Module.Param1WriteEnabled2=true
    Module.Param1WriteEnabled3=true
    Module.Param2WriteEnabled2=true
    Module.Param2WriteEnabled3=true
    Module.Param3WriteEnabled2=true
    Module.Param3WriteEnabled3=true
    Write Parameter Index 0=false
    Write Parameter Index 2=true
    Write Parameter Index 3=true
   )
   &ScaleColor(
    Scale Alpha=&Multiply_Float(
     A=&FloatFromCurve(FloatCurve=DataInterface() {propertyValues {Curve {keys [<at 0.5 0>,<at 1.5 1>]}}}, CurveIndex=$Emitter.Age)
     B=&Lerp_Float(A=0.5, B=1, Alpha=$System.FN_Integrator001.Value)
    )
    ScaleA=true
    ScaleRGB=false
   )
   /SetVariables(Particles.SecondWheelColor:LinearColor=&MultiplyLinearColor(Linear Color=<rgba 8 8 8 1>, Scale Factor=&MakeVector4_FromLinearColor(LinearColor=$Particles.Color))) {isSetParameters=true}
   /SetVariables(Particles.SecondWheelSize:Vector3f=&Multiply_Vector(A=$Particles.Scale, B=<1 1 1>)) {isSetParameters=true}
  ]
  rendererClasses [&NiagaraMeshRendererProperties]
  renderers [
   {rendererClass=&NiagaraMeshRendererProperties, rendererIndex=0}
   {rendererClass=&NiagaraMeshRendererProperties, rendererIndex=1}
  ]
 }
 SpeedLines {
  emitterUpdateScript [
   &EmitterState()
   &SpawnRate(
    SpawnRate=&MakeCustomFloatFromBool(Bool=$User.bIsFrontEndPreview, True Float=50, False Float=90)
    Use Spawn Probability=true
   ) {module=SpawnRate001}
  ]
  particleSpawnScript [
   &InitializeParticle(
    Color=&MakeLinearColorFromVectorAndFloat(
     "Vector (RGB)"=<1 1 1>
     "Float (Alpha)"=&Multiply_Float(
      A=&UniformRangedFloat(Minimum=0.1, Maximum=0.18)
      B=&MakeCustomFloatFromBool(Bool=$User.bIsFrontEndPreview, True Float=0.5, False Float=1)
     )
    )
    Lifetime Max=0.16
    Lifetime Min=0.1
    Lifetime Mode=::NewEnumerator1
    Sprite Size Max=<3 10>
    Sprite Size Min=<2 2>
    Sprite Size Mode=::NewEnumerator2
    UsePositionOffset=false
    Write Material Random=true
   )
   &TorusLocation(
    Align Torus With Axis=true
    Handle Radius=0
    Large Radius:NiagaraFloat=&RandomRangeFloat(Minimum=80, Maximum=150)
    MaskSpawnGroup=false
    Module.Override Randomness=false
    Offset=<0 -180 0>
    Torus Axis=<0 1 0>
    Torus Origin:NiagaraPosition=&Add_Vector(
     A=Linked() {linkedVariable {name=Particles.Position, type=NiagaraPosition}}
     B=&MakeVector(X=0, Y=&MakeCustomFloatFromBool(Bool=$User.bIsFrontEndPreview, True Float=0, False Float=0), Z=0)
    )
   )
   &AddVelocityFromPoint(Module.Use Falloff Distance=false, Origin Offset=<0 -500 0>, Velocity Strength=600) {module=AddVelocityFromPoint001}
   &AddVelocity(Velocity=&RandomRangeVector(Minimum=<0 300 0>, Maximum=<0 700 0>)) {module=AddVelocity001}
  ]
  particleUpdateScript [
   &SolveForcesAndVelocity(Clamp Velocity=false, Limit Acceleration=false)
   &ParticleState()
   &ScaleSpriteSize(Scale Factor=&Vector2DFromCurve(
     Vector2Curve=DataInterface() {propertyValues {XCurve {keys [<at 0 0.5 RCIM_Cubic>,<at 0.5 1 RCIM_Cubic>]}, YCurve {keys [<at 0 0 RCIM_Cubic>,<at 0.5 1 RCIM_Cubic>]}}}
     ScaleCurve=&Vector2DFromFloat(Value=1)
    ))
   &ScaleSpriteSizeBySpeed(Max Scale Factor=<1 50>, Min Scale Factor=<0.01 0.05>, Velocity Threshold=1)
   &ScaleColor(
    Scale Alpha=&FloatFromCurve(
     FloatCurve=DataInterface() {propertyValues {Curve {keys [
         <at 0 0 RCIM_Cubic>
         <at 0.3759 1.0207>
         {arriveTangent=-2.221, interpMode=RCIM_Cubic, leaveTangent=-2.221, tangentMode=RCTM_User, time=1, value=0}
        ]}}}
     Scale Curve=&MakeCustomFloatFromBool(
      Bool=$User.bIsFrontEndPreview
      True Float=&MakeCustomFloatFromBool(Bool=$User.bIsFullyDeployed, True Float=1, False Float=0)
      False Float=&Lerp_Float(A=0, B=1, Alpha=$System.FN_Integrator001.Value)
     )
    )
    Scale RGB=&VectorFromFloat(Value=1)
    ScaleA=true
    ScaleRGB=true
   )
  ]
  rendererClasses [&NiagaraSpriteRendererProperties]
  renderers [{rendererClass=&NiagaraSpriteRendererProperties}]
 }
}
summary {
 systemName=NS_Glider_BariumReboot32_AnimTrail
 emitters [
  {emitterName=SpeedLines, rendererClasses [&NiagaraSpriteRendererProperties]}
  {emitterName=NE_Wheels, rendererClasses [&NiagaraMeshRendererProperties]}
  {emitterName=FN_Ribbon_Light, rendererClasses [&NiagaraRibbonRendererProperties]}
 ]
 userVariables [
  User.BariumPaintColor {type=LinearColor, description="", defaultValue {struct=LinearColor, value=<rgba 1 1 1 1>}}
  User.bIsFrontEnd {type=NiagaraBool}
  User.bIsFrontEndPreview {type=NiagaraBool}
  User.bIsFullyDeployed {type=NiagaraBool}
  User.ForwardDot {type=NiagaraFloat, defaultValue=0, description=""}
  User.RibbonIDOffset {type=NiagaraInt32}
  User.RightDot {type=NiagaraFloat, defaultValue=0, description=""}
  User.SkeletalMesh {
   type=NiagaraDataInterfaceSkeletalMesh
   defaultValue {
    struct=NiagaraExt_VariableValue_DataInterface
    value {dataInterface=&self:NiagaraDataInterfaceSkeletalMesh_0}
   }
  }
 ]
}
